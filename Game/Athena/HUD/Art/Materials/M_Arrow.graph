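# M_Arrow is a translucent UI material for the Athena HUD that builds a procedural arrow (triangle) shape from UV math, without textures.
# Centred, Y-offset UVs are folded with Abs and weighted by X/Y ratio parameters into a diamond distance field.
# A sine-based vertical mask multiplies that shape to cut one edge.
# Gradient Triangle is the Power exponent sharpening both masks, while Gradient Top Edge is unconnected.
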Material /Game/Athena/HUD/Art/Materials/M_Arrow
graph {
 properties {blendMode=BLEND_Translucent, materialDomain=MD_UI, pixelDepthOffsetMode=PDOM_Legacy, usedShadingModels=Default Lit}
 nodes [
  n1 = Subtract(A=TextureCoordinate()) {constB=0.5}
  n2 = Add(A=ComponentMask(None=@n1) {g=true}, B=ScalarParameter() {defaultValue=0.25, parameterName=Y offset})
  n3 = Abs(None=AppendVector(A=ComponentMask(None=@n1) {r=true}, B=@n2))
  n4 = Add(
   A=Multiply(A=ComponentMask(None=@n3) {r=true}, B=ScalarParameter() {defaultValue=4, parameterName=Triangle X Ratio})
   B=Multiply(A=ComponentMask(None=@n3) {g=true}, B=ScalarParameter() {defaultValue=1.75, parameterName=Triangle Y Ratio})
  )
  n5 = ScalarParameter() {defaultValue=80, parameterName=Gradient Triangle}
  n6 = Multiply(A=Add(A=ComponentMask(None=@n2) {g=true})) {constB=0.5}
  n7 = OneMinus(None=Sine(None=Add(A=@n6)))
  Multiply(A=OneMinus(None=Power(Base=@n4, Exp=@n5)), B=Saturate(None=Power(Base=@n7, Exp=@n5)))
  ScalarParameter() {defaultValue=80, parameterName=Gradient Top Edge}
 ]
}
