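# Translucent UI material for the HUD that draws a procedural two-color gradient with an independent alpha gradient, using no textures.
# Each gradient offsets the UVs, picks an axis through the static Direction mask, and shapes it with SmoothCurve using shared Top and Bottom tangents.
# Per-gradient Taper (power) and Lift (multiply) parameters then drive lerps between Color_B and Color_A, and between two alpha values.
# Overall_Alpha scales the alpha result.
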Material /Game/Athena/HUD/Art/Materials/M_UI_HUD_SimpleGradient
graph {
 properties {blendMode=BLEND_Translucent, materialDomain=MD_UI, pixelDepthOffsetMode=PDOM_Legacy, usedShadingModels=Default Lit}
 nodes [
  n1 = ScalarParameter() {group=Color, parameterName=Offset_Color_Gradient}
  n2 = TextureCoordinate()
  n3 = MaterialFunctionCall!(
   X=StaticComponentMaskParameter(None=Add(A=@n2, B=@n1)) {defaultG=true, parameterName=Direction}
   Tangent 0=ScalarParameter() {group=Global, parameterName=BottomTangent}
   Tangent 1=ScalarParameter() {defaultValue=3, group=Global, parameterName=TopTangent}
  ) {materialFunction=/Engine/Functions/Engine_MaterialFunctions02/Gradients/SmoothCurve}
  n4 = Saturate(None=Power(Base=Saturate(None=@n3.Result), Exp=ScalarParameter() {defaultValue=1, group=Color, parameterName=Taper_Color_Gradient}))
  n5 = ScalarParameter() {defaultValue=1, group=Color, parameterName=Lift_Color_Gradient}
  LinearInterpolate(
   A=VectorParameter() {group=Color, parameterName=Color_B, defaultValue=<rgba 0 1 0 1>}.RGB
   B=VectorParameter() {group=Color, parameterName=Color_A, defaultValue=<rgba 1 0 0 1>}.RGB
   Alpha=Saturate(None=Multiply(A=@n4, B=@n5))
  )
  n6 = ScalarParameter() {defaultValue=-0.5, group=Alpha, parameterName=Offset_Alpha_Gradient}
  n7 = MaterialFunctionCall(
   X=StaticComponentMaskParameter(None=Add(A=@n2, B=@n6)) {defaultG=true, parameterName=Direction}
   Tangent 0=ScalarParameter() {group=Global, parameterName=BottomTangent}
   Tangent 1=ScalarParameter() {defaultValue=3, group=Global, parameterName=TopTangent}
  ) {materialFunction=&SmoothCurve}
  n8 = Saturate(None=Power(Base=Saturate(None=@n7.Result), Exp=ScalarParameter() {defaultValue=1, group=Alpha, parameterName=Taper_Alpha_Gradient}))
  n9 = ScalarParameter() {defaultValue=1, group=Alpha, parameterName=Lift_Alpha_Gradient}
  n10 = LinearInterpolate(
   A=ScalarParameter() {defaultValue=1, group=Alpha, parameterName=Alpha_Gradient_A}
   B=ScalarParameter() {defaultValue=1, group=Alpha, parameterName=Alpha_Gradient_B}
   Alpha=Saturate(None=Multiply(A=@n8, B=@n9))
  )
  Saturate(None=Multiply(A=Saturate(None=@n10), B=ScalarParameter() {defaultValue=1, group=Global, parameterName=Overall_Alpha}))
 ]
}
