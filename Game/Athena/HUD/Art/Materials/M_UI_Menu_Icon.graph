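# Translucent UI material for a menu icon, drawn from an MSDF texture that defaults to the sidebar menu icon.
# UI UVs are offset, scaled around the center and rotated by parameters before sampling the Icon texture.
# The sample is converted to a signed distance field and thresholded with adjustable smoothing into an anti-aliased mask.
# Multiply nodes apply an Opacity parameter and vertex alpha.
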Material /Game/Athena/HUD/Art/Materials/M_UI_Menu_Icon
graph {
 properties {bAutomaticallySetUsageInEditor=false, blendMode=BLEND_Translucent, materialDomain=MD_UI, usedShadingModels=Default Lit}
 nodes [
  Multiply(A=NamedRerouteUsage(), B=ScalarParameter() {defaultValue=1, group=1 - Transform, parameterName=Opacity, sliderMax=1})
  Multiply(A=VertexColor().A, B=NamedRerouteUsage())
  n1 = TextureSampleParameter2D(UVs=NamedRerouteUsage()) {group=2 - Distance Field, parameterName=Icon, samplerType=SAMPLERTYPE_LinearColor, texture=/Game/Athena/HUD/Art/T_UI_Icon_Sidebar_Menu_MSDF}
  NamedRerouteDeclaration(None=MaterialFunctionCall(
    DistanceField=MaterialFunctionCall!(MSDF RGB=@n1.RGB) {materialFunction=/FortUILibrary/MaterialFunctions/Shapes/Modifiers/MF_UI_MSDFtoSDF}.SDF
    Threshold=Constant() {r=0.5}
    Smoothing=ScalarParameter() {defaultValue=0.07, group=2 - Distance Field, parameterName=Smoothing}
   ) {materialFunction=&MF_UI_AntiAliasing}.Outer) {name=Icon Mask, nodeColor=<rgba 0.9688 0 1 1>}
  n2 = MaterialFunctionCall() {materialFunction=&GetUserInterfaceUV}
  n3 = AppendVector(
   A=Subtract(A=ComponentMask(None=@n2.Normalized UV) {r=true}, B=ScalarParameter() {group=1 - Transform, parameterName=OffsetX, sortPriority=48})
   B=Add(A=ComponentMask(None=@n2.Normalized UV) {g=true}, B=ScalarParameter() {group=1 - Transform, parameterName=OffsetY, sortPriority=48})
  )
  n4 = ScalarParameter() {defaultValue=1, group=1 - Transform, parameterName=Scale}
  NamedRerouteDeclaration(None=MaterialFunctionCall(
    UVs=MaterialFunctionCall(UVs=@n3, Texture Scale=@n4) {materialFunction=&ScaleUVsByCenter}.UVs
    "Rotation Angle (0-1)"=ScalarParameter() {group=1 - Transform, parameterName=Rotation}
   ) {materialFunction=&CustomRotator}.Rotated Values) {name=UVs, nodeColor=<rgba 0 0 0 1>}
 ]
}
