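# Translucent UI material for a HUD modal background with a two-color gradient and a radial expand reveal.
# Pixel-size aspect correction keeps gradients circular; one centered at the UV origin blends RightColor to LeftColor and their alphas.
# Scale grows a second gradient's radius and density from near that corner, masking the alpha.
# Static switch UseAnimatedHorizontalFade replaces that alpha with a hard horizontal wipe driven by AnimatedScalar.
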Material /Game/Athena/HUD/Art/Materials/M_UI_ModalRadialExpand
graph {
 properties {blendMode=BLEND_Translucent, materialDomain=MD_UI, pixelDepthOffsetMode=PDOM_Legacy, usedShadingModels=Default Lit}
 nodes [
  n1 = MaterialFunctionCall() {materialFunction=&GetUserInterfaceUV}
  n2 = MaterialFunctionCall(Float2=@n1.Pixel Size) {materialFunction=&BreakOutFloat2Components}
  n3 = Max(A=@n2.R)
  n4 = Multiply(A=AppendVector(A=Constant() {r=1}, B=Divide(A=@n2.G, B=@n3)), B=@n1.Normalized UV)
  n5 = Distance(A=Constant2Vector(), B=@n1.Pixel Size)
  n6 = Multiply(A=MaterialFunctionCall(UVs=@n4, CenterPosition=Constant2Vector(), Radius=Divide(A=@n5, B=@n3), Density=Constant() {r=1.666}) {materialFunction=&RadialGradientExponential}.RadialGradientExponential)
  LinearInterpolate(
   A=VectorParameter() {parameterName=RightColor, defaultValue=<rgba 0.08 0.38 0.8 1>}.RGB
   B=VectorParameter() {parameterName=LeftColor, defaultValue=<rgba 0.06 0.195 0.6 1>}.RGB
   Alpha=@n6
  )
  n7 = Add(
   A=ComponentMask(None=TextureCoordinate()) {r=true}
   B=LinearInterpolate(Alpha=ScalarParameter() {parameterName=AnimatedScalar}) {constA=1, constB=0}
  )
  n8 = ScalarParameter() {defaultValue=0.95, parameterName=RightAlpha}
  n9 = ScalarParameter() {defaultValue=1, parameterName=LeftAlpha}
  n10 = ScalarParameter() {defaultValue=1, parameterName=Scale}
  n11 = MaterialFunctionCall(
   UVs=@n4
   CenterPosition=Constant2Vector() {g=0.0125, r=0.075}
   Radius=LinearInterpolate(A=Constant(), B=Constant() {r=1}, Alpha=@n10)
   Density=LinearInterpolate(A=Constant(), B=Constant() {r=512}, Alpha=@n10)
  ) {materialFunction=&RadialGradientExponential}
  StaticSwitchParameter(
   True=OneMinus(None=Floor(None=@n7))
   False=Multiply(A=LinearInterpolate(A=@n8, B=@n9, Alpha=@n6), B=Clamp(None=@n11.RadialGradientExponential))
  ) {parameterName=UseAnimatedHorizontalFade}
 ]
}
