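# Translucent UI master material that draws a no-build indicator: the build-mode icon inside a filled ring crossed by a diagonal slash.
# Exponential radial gradients form the ring and clip the outer circle, while rotated, thresholded UV half-planes intersect into the slash.
# A wider copy of the slash mask cuts a gap in the scaled icon.
# Parameters set colors and alphas, and AnimateIconColor blends ring and slash toward RingColor_Animated.
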Material /Game/Athena/HUD/Art/Materials/M_UI_NoBuildIndicator_Master
graph {
 properties {blendMode=BLEND_Translucent, materialDomain=MD_UI, pixelDepthOffsetMode=PDOM_Legacy, usedShadingModels=Default Lit}
 nodes [
  n1 = ScalarParameter() {defaultValue=-0.125, parameterName=CrossRotation}
  Add(A=@n1) {constB=0}
  n2 = OneMinus(None=MaterialFunctionCall(
    Radius=ScalarParameter() {defaultValue=0.425, group=Ring, parameterName=InnerRadius}
    Density=ScalarParameter() {defaultValue=24, group=Ring, parameterName=InnerHardness}
   ) {materialFunction=&RadialGradientExponential}.RadialGradientExponential)
  n3 = LinearInterpolate(
   A=ScalarParameter() {defaultValue=0.5, group=Fill, parameterName=FillAlpha}
   B=ScalarParameter() {defaultValue=1, group=Ring, parameterName=RingAlpha}
   Alpha=@n2
  )
  n4 = MaterialFunctionCall(UVs=TextureCoordinate(), "Rotation Angle (0-1)"=@n1) {materialFunction=&CustomRotator}
  n5 = Add(A=@n4.Rotated Values) {constB=-0.5}
  n6 = ScalarParameter() {defaultValue=0.0625, parameterName=Cross_Thickness}
  n7 = ScalarParameter() {defaultValue=32, parameterName=Cross_Hardness}
  n8 = Multiply(A=ComponentMask(None=Add(A=@n5, B=@n6) {constB=-0.5}) {r=true}, B=@n7)
  n9 = ScalarParameter() {defaultValue=-0.0125, parameterName=CrossRotation_2}
  n10 = Add(A=OneMinus(None=MaterialFunctionCall(UVs=@n4.Rotated Values, "Rotation Angle (0-1)"=@n9) {materialFunction=&CustomRotator}.Rotated Values)) {constB=-0.5}
  n11 = Multiply(A=ComponentMask(None=Add(A=@n10, B=@n6) {constB=-0.5}) {r=true}, B=@n7)
  n12 = Saturate(None=DotProduct(A=Saturate(None=@n8), B=Saturate(None=@n11)))
  n13 = MaterialFunctionCall(UVs=TextureCoordinate(), Texture Scale=ScalarParameter() {defaultValue=0.75, group=Icon, parameterName=IconScale}) {materialFunction=&ScaleUVsByCenter}
  n14 = ScalarParameter() {defaultValue=0.125, parameterName=Crop_Thickness}
  n15 = ScalarParameter() {defaultValue=32, parameterName=Cross_Hardness}
  n16 = Multiply(A=ComponentMask(None=Add(A=@n5, B=@n14) {constB=-0.5}) {r=true}, B=@n15)
  n17 = Multiply(A=ComponentMask(None=Add(A=@n10, B=@n14) {constB=-0.5}) {r=true}, B=@n15)
  n18 = Saturate(None=DotProduct(A=Saturate(None=@n16), B=Saturate(None=@n17)))
  n19 = Saturate(None=DotProduct(A=TextureSample(UVs=@n13.UVs) {texture=/Game/UI/Icons/Icon-QuickbarMode-Build-M}.A, B=OneMinus(None=@n18)))
  DotProduct(
   A=MaterialFunctionCall(Density=ScalarParameter() {defaultValue=20, group=Ring, parameterName=OuterHardness}) {materialFunction=&RadialGradientExponential}.RadialGradientExponential
   B=LinearInterpolate(A=LinearInterpolate(A=@n3, Alpha=@n12), B=ScalarParameter() {defaultValue=1, group=Icon, parameterName=IconAlpha}, Alpha=@n19)
  )
  n20 = VectorParameter() {group=Ring, parameterName=RingColor, defaultValue=<rgba 1 1 1 1>}
  n21 = VectorParameter() {group=Ring, parameterName=RingColor_Animated, defaultValue=<rgba 1 1 1 1>}
  n22 = ScalarParameter() {parameterName=AnimateIconColor}
  LinearInterpolate(
   A=LinearInterpolate(
    A=VectorParameter() {group=Fill, parameterName=FillColor, defaultValue=<rgba 0 0 0 1>}.RGB
    B=LinearInterpolate(A=@n20.RGB, B=@n21.RGB, Alpha=@n22)
    Alpha=LinearInterpolate(A=@n2, Alpha=@n12)
   )
   B=VectorParameter() {group=Icon, parameterName=IconColor, defaultValue=<rgba 1 1 1 1>}.RGB
   Alpha=@n19
  )
 ]
}
