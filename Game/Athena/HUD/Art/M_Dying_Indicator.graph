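# UI translucent material for the squad DBNO (down but not out) indicator, showing a pulsing red icon.
# A time-driven sine blends between two reds that tint the blurred icon, and the cross texture is lerped over it using its own RGB as mask.
# The larger of the two texture alphas forms a combined mask.
# Shine parameters build a timed on/off cycle with Fmod and If, separate from that color blend.
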Material /Game/Athena/HUD/Art/M_Dying_Indicator
graph {
 properties {blendMode=BLEND_Translucent, materialDomain=MD_UI, pixelDepthOffsetMode=PDOM_Legacy, usedShadingModels=Default Lit}
 nodes [
  n1 = ScalarParameter() {defaultValue=12, group=Shine, parameterName=Shine Speed}
  n2 = Divide(A=ScalarParameter() {defaultValue=2, group=Shine, parameterName=Cycles On}, B=@n1)
  n3 = Constant() {r=2}
  n4 = Multiply(A=Divide(A=ScalarParameter() {defaultValue=8, group=Shine, parameterName=Cycles Off}, B=@n1), B=Constant() {r=2})
  n5 = Time()
  n6 = Fmod(A=@n5, B=Add(A=Multiply(A=@n2, B=@n3), B=@n4))
  Add(B=MaterialFunctionCall(Number=Sine(None=@n6)) {materialFunction=&DebugScalarValues}.Result)
  Clamp(None=@n6)
  If(A=@n6, B=Constant() {r=1})
  n7 = Constant3Vector() {constant=<rgba 1 0 0 0>}
  n8 = Constant3Vector() {constant=<rgba 0.7 0 0 1>}
  n9 = MaterialFunctionCall(Speed=Constant() {r=0.5}) {materialFunction=&DebugTimeSine}
  n10 = TextureSample() {texture=&/squad_DBNOIcon_Blur}
  n11 = TextureSample() {texture=&/squad_DBNOIcon_Cross}
  LinearInterpolate(A=Multiply(A=LinearInterpolate(A=@n7, B=@n8, Alpha=@n9.Result), B=@n10.RGB), B=@n11.RGB, Alpha=@n11.RGB)
  MaterialFunctionCall(Speed=Constant() {r=2}) {materialFunction=&DebugTimeSine}
  Max(A=@n11.A, B=@n10.A)
  n12 = Constant() {r=1}
  n13 = Constant() {r=-1}
  n14 = If(A=@n6, B=@n4, "A > B"=Constant(), "A == B"=@n13, "A < B"=@n13)
  Reroute(None=Add(A=Fmod(A=@n5, B=@n12), B=@n14) {constB=-1})
  Sine(None=@n14)
 ]
}
