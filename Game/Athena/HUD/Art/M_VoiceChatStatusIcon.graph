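# Translucent UI material for the HUD voice chat status icon, showing idle, speaking or muted microphone art.
# Each of three mic textures becomes a float4 from its red channel, except the idle icon's alpha, which is its texture alpha at half strength.
# IsSpeaking lerps idle to active, then IsMuted lerps that to muted, overriding speech at full value.
# The blended float4 is then split into components.
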Material /Game/Athena/HUD/Art/M_VoiceChatStatusIcon
graph {
 properties {blendMode=BLEND_Translucent, materialDomain=MD_UI, pixelDepthOffsetMode=PDOM_Legacy, usedShadingModels=Default Lit}
 nodes [
  n1 = TextureSample() {texture=&/mic-notalk}
  n2 = MaterialFunctionCall(X=@n1.R, Y=@n1.R, Z=@n1.R, A=Multiply(A=@n1.A) {constB=0.5}) {materialFunction=&MakeFloat4}
  n3 = TextureSample() {texture=&/mic-active}
  n4 = MaterialFunctionCall(X=@n3.R, Y=@n3.R, Z=@n3.R, A=@n3.R) {materialFunction=&MakeFloat4}
  n5 = ScalarParameter() {parameterName=IsSpeaking, sliderMax=1}
  n6 = TextureSample() {texture=&/mic-muted}
  MaterialFunctionCall(Float4=LinearInterpolate(
    A=LinearInterpolate(A=@n2.Result, B=@n4.Result, Alpha=@n5)
    B=MaterialFunctionCall(X=@n6.R, Y=@n6.R, Z=@n6.R, A=@n6.R) {materialFunction=&MakeFloat4}.Result
    Alpha=ScalarParameter() {parameterName=IsMuted}
   )) {materialFunction=&BreakOutFloat4Components}
 ]
}
