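# Translucent UI material for a Creative logo in the Athena HUD, built from the Creative move tool's deselected icon texture.
# The icon's red channel is multiplied by an Alpha scalar parameter, faint by default, to form a mask.
# The same channel drives a lerp whose two inputs are the same white constant, so the color stays flat white.
# A light blue constant is left unconnected.
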Material /Game/Athena/HUD/Art/Materials/M_UI_CreativeLogo
graph {
 properties {blendMode=BLEND_Translucent, materialDomain=MD_UI, pixelDepthOffsetMode=PDOM_Legacy, usedShadingModels=Default Lit}
 nodes [
  Constant3Vector() {constant=<rgba 0 0.6667 1 0>}
  n1 = Constant3Vector() {constant=<rgba 1 1 1 0>}
  n2 = TextureSample(UVs=TextureCoordinate()) {texture=/Game/Weapons/Prototype/ObjectMover/Textures/T_CreativeMoveTool_DeselectedIcon}
  LinearInterpolate(A=@n1, B=@n1, Alpha=@n2.R)
  Multiply(A=@n2.R, B=ScalarParameter() {defaultValue=0.2, parameterName=Alpha})
 ]
}
