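# UI material for the Creative start panel in the Athena HUD, named as the color for its description area.
# It is a translucent User Interface domain material exposing a Color vector parameter, defaulting to bright blue, and a fully opaque Alpha scalar parameter.
# The graph holds only these two parameter nodes, with no textures or math, and shows no wiring to the material outputs.
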Material /Game/Athena/HUD/Art/Materials/M_UI_CreativeStartPanel_DescriptionColor
graph {
 properties {blendMode=BLEND_Translucent, materialDomain=MD_UI, pixelDepthOffsetMode=PDOM_Legacy, usedShadingModels=Default Lit}
 nodes [
  ScalarParameter() {defaultValue=1, parameterName=Alpha}
  VectorParameter() {parameterName=Color, defaultValue=<rgba 0.1 0.4 1 1>}
 ]
}
